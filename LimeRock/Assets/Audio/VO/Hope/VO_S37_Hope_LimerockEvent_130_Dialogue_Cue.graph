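# SoundCue that plays Hope's voice line 130 for the Season 37 LimeRock event.
# A single dialogue player resolves the DialogueWave with the LimeRock dialogue voice as speaker and no targets.
# It routes through the Events VO sound class and borrows the audio-log attenuation asset, with a very large max distance.
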SoundCue /LimeRock/Assets/Audio/VO/Hope/VO_S37_Hope_LimerockEvent_130_Dialogue_Cue
root=DialoguePlayer_0
nodes [{
  id=DialoguePlayer_0
  properties {dialogueWaveParameter {
    dialogueWave=/LimeRock/Assets/Audio/VO/Hope/Waves/VO_S37_Hope_LimerockEvent_130_Dialogue
    context {speaker=/LimeRock/Assets/Audio/VO/LimeRock_DV, targets []}
   }}
 }]
properties {
 attenuationSettings=/Game/Sounds/Fort_Audio_Blueprints/Utilities/EditorUtilityWidgets/VO_Utilities/Assets/AS_Fort_AudioLogs
 maxDistance=2097152
 soundClassObject=/Game/Sounds/FortSoundClassesAndModes/Fort_Events_VO
 volumeMultiplier=1
}
